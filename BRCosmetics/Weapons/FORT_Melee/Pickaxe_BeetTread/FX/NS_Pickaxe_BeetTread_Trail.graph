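# Niagara system for the BeetTread pickaxe's melee trail, drawn by a single ribbon emitter.
# FN anim-trail modules spawn, initialize and update short-lived, pale translucent particles from bone indices on the system's skeletal mesh.
# ScaleColor fades alpha and darkens RGB over curves, with time-of-day inputs that shift the color toward a dimmer blue at night.
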NiagaraSystem /BRCosmetics/Weapons/FORT_Melee/Pickaxe_BeetTread/FX/NS_Pickaxe_BeetTread_Trail
dependencies {
 usedDataInterfaces [&NiagaraDataInterfaceColorCurve, &NiagaraDataInterfaceCurve, &NiagaraDataInterfaceSkeletalMesh]
 usedDynamicInputs []
 usedModules [
  &self:SystemSpawnScript.SystemScriptSource.SystemScriptGraph.NiagaraNodeAssignment_0.SetVariables_#1
  &SystemState
 ]
}
emitters {Trail {
  emitterUpdateScript [
   &EmitterState(Life Cycle Mode=::NewEnumerator1, Loop Behavior=::NewEnumerator1, Loop Duration=0.3, UseLoopDelay=false)
   &FN_SpawnAnimTrail()
  ]
  particleSpawnScript [
   &FN_InitializeAnimTrailBoneIndecies()
   &FN_InitializeAnimTrail(Color=<rgba 0.8564 0.8321 0.9583 0.2>, Lifetime=0.3, SkeletalMesh=$System.Skeletal Mesh)
  ]
  particleUpdateScript [
   &ParticleState()
   &FN_UpdateAnimTrail(SkeletalMesh=$System.Skeletal Mesh)
   &ScaleColor(
    Scale Alpha=&FloatFromCurve(
     FloatCurve=DataInterface() {propertyValues {Curve {keys [
         {arriveTangent=-1.1204, interpMode=RCIM_Cubic, leaveTangent=-1.1204, tangentMode=RCTM_User, time=0, value=1}
         {interpMode=RCIM_Cubic, tangentMode=RCTM_User, time=1, value=0}
        ]}}}
     Scale Curve=&FNTimeOfDayScalar(Night Scale=1, Morning Scale=1, Evening Scale=1)
    )
    Scale RGB=&Multiply_Vector(
     A=&MakeVectorFromLinearColorRGB(Color=&ColorFromCurve(ColorCurve=DataInterface() {propertyValues {
         AlphaCurve {keys []}
         BlueCurve {keys [<at 0 0.8854 RCIM_Cubic>,<at 0.7289 0.4115 RCIM_Cubic>]}
         GreenCurve {keys [<at 0 0.8854 RCIM_Cubic>,<at 0.7289 0.4115 RCIM_Cubic>]}
         RedCurve {keys [<at 0 0.8854 RCIM_Cubic>,<at 0.7289 0.4115 RCIM_Cubic>]}
        }}))
     B=&FN_Select_Vector_from_Time_Of_Day(Morning=<1 1 1>, Day=<1 1 1>, Evening=<1 1 1>, Night=<0.5 0.5 1>)
    )
    ScaleA=true
    ScaleRGB=true
   )
  ]
  rendererClasses [&NiagaraRibbonRendererProperties]
  renderers [{rendererClass=&NiagaraRibbonRendererProperties}]
 }}
summary {
 systemName=NS_Pickaxe_BeetTread_Trail
 emitters [{emitterName=Trail, rendererClasses [&NiagaraRibbonRendererProperties]}]
 userVariables [User.RibbonIDOffset {type=NiagaraInt32}]
}
