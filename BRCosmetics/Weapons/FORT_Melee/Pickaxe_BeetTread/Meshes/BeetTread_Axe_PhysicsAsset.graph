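# Physics asset for the BeetTread two-handed pickaxe, setting up two dangling rope chains, top and lower, on the handle.
# A capsule body covers the handle, and each chain has a master capsule plus three segment capsules.
# Each master is angularly locked to the handle, and the segments link in sequence with limited swing and tight twist.
# The top chain has narrower swing limits than the lower one.
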PhysicsAsset /BRCosmetics/Weapons/FORT_Melee/Pickaxe_BeetTread/Meshes/BeetTread_Axe_PhysicsAsset
bodies {
 Melee_TwoHanded_Handle [{shapeType=Capsule, length=67.5496, radius=2.4, center=<1.7404 -1.1623 15.0794>, rotation=<rot -0.0986 -90 -0.0334>}]
 dyn_ropes_lower_01 [{shapeType=Capsule, length=2, radius=1, shapeName=dyn_ropes_lower_01_capsule, center=<3.491 0 0>, rotation=<rot 90 0 0>}]
 dyn_ropes_lower_02 [{shapeType=Capsule, length=4.7272, radius=1.8533, shapeName=dyn_ropes_lower_02_capsule, center=<3.491 0 0>, rotation=<rot 90 0 0>}]
 dyn_ropes_lower_03 [{shapeType=Capsule, length=2.5137, radius=2.2776, shapeName=dyn_ropes_lower_03_capsule, center=<3.491 0 0>, rotation=<rot 90 0 0>}]
 dyn_ropes_lower_MSTR [{shapeType=Capsule, length=1, radius=1, shapeName=dyn_ropes_lower_MSTR_capsule, rotation=<rot 0 -90 0>}]
 dyn_ropes_top_01 [{shapeType=Capsule, length=2, radius=1, shapeName=dyn_ropes_top_01_capsule, center=<3.491 0 0>, rotation=<rot 90 0 0>}]
 dyn_ropes_top_02 [{shapeType=Capsule, length=4.7272, radius=1.8533, shapeName=dyn_ropes_top_02_capsule, center=<3.491 0 0>, rotation=<rot 90 0 0>}]
 dyn_ropes_top_03 [{shapeType=Capsule, length=2.5137, radius=2.2776, shapeName=dyn_ropes_top_03_capsule, center=<3.491 0 0>, rotation=<rot 90 0 0>}]
 dyn_ropes_top_MSTR [{shapeType=Capsule, length=1, radius=1, shapeName=dyn_ropes_top_MSTR_capsule, rotation=<rot 0 -90 0>}]
}
constraints [
 {
  bone1Name=dyn_ropes_lower_MSTR
  bone2Name=Melee_TwoHanded_Handle
  swing1LimitDegrees=45
  swing1Motion=Locked
  swing2LimitDegrees=45
  swing2Motion=Locked
  twistLimitDegrees=45
  twistMotion=Locked
 }
 {bone1Name=dyn_ropes_lower_01, bone2Name=dyn_ropes_lower_MSTR, swing1LimitDegrees=90, swing2LimitDegrees=90, twistLimitDegrees=5}
 {bone1Name=dyn_ropes_lower_02, bone2Name=dyn_ropes_lower_01, swing1LimitDegrees=45, swing2LimitDegrees=45, twistLimitDegrees=5}
 {bone1Name=dyn_ropes_lower_03, bone2Name=dyn_ropes_lower_02, swing1LimitDegrees=45, swing2LimitDegrees=45, twistLimitDegrees=5}
 {
  bone1Name=dyn_ropes_top_MSTR
  bone2Name=Melee_TwoHanded_Handle
  swing1LimitDegrees=45
  swing1Motion=Locked
  swing2LimitDegrees=45
  swing2Motion=Locked
  twistLimitDegrees=45
  twistMotion=Locked
 }
 {bone1Name=dyn_ropes_top_01, bone2Name=dyn_ropes_top_MSTR, swing1LimitDegrees=45, swing2LimitDegrees=45, twistLimitDegrees=5}
 {bone1Name=dyn_ropes_top_02, bone2Name=dyn_ropes_top_01, swing1LimitDegrees=35, swing2LimitDegrees=35, twistLimitDegrees=5}
 {bone1Name=dyn_ropes_top_03, bone2Name=dyn_ropes_top_02, swing1LimitDegrees=35, swing2LimitDegrees=35, twistLimitDegrees=5}
]
