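# Animation Blueprint for the BeetTread axe pickaxe that adds physics-driven sway to its top and lower ropes.
# When bRigidBodyEnabled is true, additive master-bone offsets, chosen by whether it is in the front end, feed a RigidBody node simulating relative to the handle.
# Otherwise, six chained, handle-relative AnimDynamics nodes drive the ropes, with angular springs and spherical limits.
# Both branches start from the cached input pose.
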AnimBlueprint /BRCosmetics/Weapons/FORT_Melee/Pickaxe_BeetTread/Meshes/BeetTread_Axe_Skeleton_AnimBlueprint
target_skeleton=&/BeetTread_Axe_Skeleton
graphs [
 AnimGraph {nodes [
   n1 = Anim Dynamics - Bone: dyn_ropes_lower_01(ComponentPose=Local To Component(LocalPose=Use cached pose 'Base_pose'()), Alpha=1, ExternalForce="0,0,0") {
    angularDampingOverride=0.9999
    angularSpringConstant=50
    bAngularSpring=true
    bOverrideAngularDamping=true
    bUseGravityOverride=true
    simulationSpace=BoneRelative
    alphaBoolBlend {blendOption=Linear}
    boundBone {boneName=dyn_ropes_lower_01}
    componentAppliedLinearAccClamp=<2000 2000 2000>
    componentLinearAccScale=<0.4 0.3 0.3>
    componentLinearVelScale=<0.4 0.3 0.3>
    gravityOverride {z=-980}
    physicsBodyDefinitions [{
      boundBone {boneName=dyn_ropes_lower_01}
      boxExtents=<10 10 10>
      constraintSetup {angularLimitsMax=<5 80 60>, angularLimitsMin=<5 -80 -80>, angularTarget=<0 0 0>}
      localJointOffset=<1 0 0>
     }]
    relativeSpaceBone {boneName=Melee_TwoHanded_Handle}
   }
   n2 = Anim Dynamics - Bone: dyn_ropes_lower_02(ComponentPose=@n1, Alpha=1, ExternalForce="0,0,0") {
    angularDampingOverride=0.9999
    angularSpringConstant=50
    bAngularSpring=true
    bOverrideAngularDamping=true
    bUseGravityOverride=true
    bUseSphericalLimits=true
    simulationSpace=BoneRelative
    alphaBoolBlend {blendOption=Linear}
    boundBone {boneName=dyn_ropes_lower_02}
    componentAppliedLinearAccClamp=<2000 2000 2000>
    componentLinearAccScale=<0.4 0.3 0.3>
    componentLinearVelScale=<0.4 0.3 0.3>
    gravityOverride {z=-980}
    physicsBodyDefinitions [{
      boundBone {boneName=dyn_ropes_lower_02}
      boxExtents=<10 10 10>
      constraintSetup {angularLimitsMax=<5 45 20>, angularLimitsMin=<5 -45 -45>, angularTarget=<0 0 0>}
      localJointOffset=<1 0 0>
     }]
    relativeSpaceBone {boneName=Melee_TwoHanded_Handle}
    sphericalLimits [
     {limitRadius=0, drivingBone {boneName=None}, sphereLocalOffset=<0 0 0>}
     {limitRadius=3, drivingBone {boneName=Melee_TwoHanded_Handle}, sphereLocalOffset=<0 -0.579 35.3426>}
    ]
   }
   n3 = Anim Dynamics - Bone: dyn_ropes_lower_03(ComponentPose=@n2, Alpha=1, ExternalForce="0,0,0") {
    angularDampingOverride=0.9999
    angularSpringConstant=50
    bAngularSpring=true
    bOverrideAngularDamping=true
    bUseGravityOverride=true
    bUseSphericalLimits=true
    simulationSpace=BoneRelative
    alphaBoolBlend {blendOption=Linear}
    boundBone {boneName=dyn_ropes_lower_03}
    componentAppliedLinearAccClamp=<2000 2000 2000>
    componentLinearAccScale=<0.4 0.3 0.3>
    componentLinearVelScale=<0.4 0.3 0.3>
    gravityOverride {z=-980}
    physicsBodyDefinitions [{
      boundBone {boneName=dyn_ropes_lower_03}
      boxExtents=<10 10 10>
      constraintSetup {angularLimitsMax=<5 45 45>, angularLimitsMin=<5 -45 -45>, angularTarget=<0 0 0>}
      localJointOffset=<1 0 0>
     }]
    relativeSpaceBone {boneName=Melee_TwoHanded_Handle}
    sphericalLimits [
     {limitRadius=3, drivingBone {boneName=Melee_TwoHanded_Handle}, sphereLocalOffset=<0 -1.0246 33.232>}
     {limitRadius=2, drivingBone {boneName=dyn_ropes_top_03}, sphereLocalOffset=<1.6732 0 0>}
    ]
   }
   n4 = Anim Dynamics - Bone: dyn_ropes_top_01(ComponentPose=@n3, Alpha=1, ExternalForce="0,0,0") {
    angularDampingOverride=0.9999
    angularSpringConstant=50
    bAngularSpring=true
    bOverrideAngularDamping=true
    bUseGravityOverride=true
    simulationSpace=BoneRelative
    alphaBoolBlend {blendOption=Linear}
    boundBone {boneName=dyn_ropes_top_01}
    componentAppliedLinearAccClamp=<2000 2000 2000>
    componentLinearAccScale=<0.4 0.3 0.3>
    componentLinearVelScale=<0.4 0.3 0.3>
    gravityOverride {z=-980}
    physicsBodyDefinitions [{
      boundBone {boneName=dyn_ropes_top_01}
      boxExtents=<10 10 10>
      constraintSetup {angularLimitsMax=<5 80 80>, angularLimitsMin=<5 -80 -60>, angularTarget=<0 0 0>}
      localJointOffset=<1 0 0>
     }]
    relativeSpaceBone {boneName=Melee_TwoHanded_Handle}
   }
   n5 = Anim Dynamics - Bone: dyn_ropes_top_02(ComponentPose=@n4, Alpha=1, ExternalForce="0,0,0") {
    angularDampingOverride=0.9999
    angularSpringConstant=50
    bAngularSpring=true
    bOverrideAngularDamping=true
    bUseGravityOverride=true
    bUseSphericalLimits=true
    simulationSpace=BoneRelative
    alphaBoolBlend {blendOption=Linear}
    boundBone {boneName=dyn_ropes_top_02}
    componentAppliedLinearAccClamp=<2000 2000 2000>
    componentLinearAccScale=<0.4 0.3 0.3>
    componentLinearVelScale=<0.4 0.3 0.3>
    gravityOverride {z=-980}
    physicsBodyDefinitions [{
      boundBone {boneName=dyn_ropes_top_02}
      boxExtents=<10 10 10>
      constraintSetup {angularLimitsMax=<5 45 45>, angularLimitsMin=<5 -45 -20>, angularTarget=<0 0 0>}
      localJointOffset=<1 0 0>
     }]
    relativeSpaceBone {boneName=Melee_TwoHanded_Handle}
    sphericalLimits [
     {limitRadius=0, drivingBone {boneName=None}, sphereLocalOffset=<0 0 0>}
     {limitRadius=3, drivingBone {boneName=Melee_TwoHanded_Handle}, sphereLocalOffset=<0 -0.9094 36.6221>}
    ]
   }
   n6 = Anim Dynamics - Bone: dyn_ropes_top_03(ComponentPose=@n5, Alpha=1, ExternalForce="0,0,0") {
    angularDampingOverride=0.9999
    angularSpringConstant=50
    bAngularSpring=true
    bOverrideAngularDamping=true
    bUseGravityOverride=true
    bUseSphericalLimits=true
    simulationSpace=BoneRelative
    alphaBoolBlend {blendOption=Linear}
    boundBone {boneName=dyn_ropes_top_03}
    componentAppliedLinearAccClamp=<2000 2000 2000>
    componentLinearAccScale=<0.4 0.3 0.3>
    componentLinearVelScale=<0.4 0.3 0.3>
    gravityOverride {z=-980}
    physicsBodyDefinitions [{
      boundBone {boneName=dyn_ropes_top_03}
      boxExtents=<10 10 10>
      constraintSetup {angularLimitsMax=<5 45 45>, angularLimitsMin=<5 -45 -45>, angularTarget=<0 0 0>}
      localJointOffset=<1 0 0>
     }]
    relativeSpaceBone {boneName=Melee_TwoHanded_Handle}
    sphericalLimits [
     {limitRadius=3, drivingBone {boneName=Melee_TwoHanded_Handle}, sphereLocalOffset=<0 -0.7852 33.5611>}
     {limitRadius=2, drivingBone {boneName=dyn_ropes_lower_03}, sphereLocalOffset=<1.6571 0 0>}
    ]
   }
   n7 = Get not in FE() -> (not in FE:bool)
   n8 = "Transform (Modify) Bone - Bone: dyn_ropes_lower_MSTR"!(
    ComponentPose=Local To Component(LocalPose=Use cached pose 'Base_pose'())
    bAlphaBoolEnabled:bool=@n7
    Translation="0,1,0"
    Rotation="10,30,0"
    Scale="1,1,1"
   ) -> (Pose) {
    alphaInputType=Bool
    rotationMode=BMM_Additive
    rotationSpace=BCS_BoneSpace
    translationMode=BMM_Additive
    translationSpace=BCS_BoneSpace
    alphaBoolBlend {blendOption=Linear}
    boneToModify {boneName=dyn_ropes_lower_MSTR}
    rotation {pitch=10, yaw=30}
    translation {y=1}
   }
   n9 = "Transform (Modify) Bone - Bone: dyn_ropes_top_MSTR"!(ComponentPose=@n8, bAlphaBoolEnabled:bool=@n7, Translation="0,0,0", Rotation="0,15,0", Scale="1,1,1") -> (Pose) {
    alphaInputType=Bool
    rotationMode=BMM_Additive
    rotationSpace=BCS_BoneSpace
    translationMode=BMM_Additive
    translationSpace=BCS_BoneSpace
    alphaBoolBlend {blendOption=Linear}
    boneToModify {boneName=dyn_ropes_top_MSTR}
    rotation {yaw=15}
   }
   n10 = Get bIsInFrontEnd()
   n11 = "Transform (Modify) Bone - Bone: dyn_ropes_lower_MSTR"(ComponentPose=@n9, bAlphaBoolEnabled=@n10, Translation="0,1,0", Rotation="10,50,0", Scale="1,1,1") {
    alphaInputType=Bool
    rotationMode=BMM_Additive
    rotationSpace=BCS_BoneSpace
    translationMode=BMM_Additive
    translationSpace=BCS_BoneSpace
    alphaBoolBlend {blendOption=Linear}
    boneToModify {boneName=dyn_ropes_lower_MSTR}
    rotation {pitch=10, yaw=50}
    translation {y=1}
   }
   n12 = "Transform (Modify) Bone - Bone: dyn_ropes_top_MSTR"(ComponentPose=@n11, bAlphaBoolEnabled=@n10, Translation="0,0,0", Rotation="0,10,0", Scale="1,1,1") {
    alphaInputType=Bool
    rotationMode=BMM_Additive
    rotationSpace=BCS_BoneSpace
    translationMode=BMM_Additive
    translationSpace=BCS_BoneSpace
    alphaBoolBlend {blendOption=Linear}
    boneToModify {boneName=dyn_ropes_top_MSTR}
    rotation {yaw=10}
   }
   n13 = RigidBody(ComponentPose=@n12) {
    bClampLinearTranslationLimitToRefPose=true
    bOverrideWorldGravity=true
    overridePhysicsAsset=&/BeetTread_Axe_PhysicsAsset
    simulationSpace=BaseBoneSpace
    alphaBoolBlend {blendOption=Linear}
    baseBoneRef {boneName=Melee_TwoHanded_Handle}
    componentAppliedLinearAccClamp=<1000 1000 1000>
    componentLinearAccScale=<0.3 0.3 0.3>
    componentLinearVelScale=<0.35 0.35 0.35>
    overrideWorldGravity {z=-980}
   }
   Output Pose(Result=Blend Poses by bool(BlendPose_0=Component To Local(ComponentPose=@n13), BlendPose_1=Component To Local(ComponentPose=@n6), bActiveValue=Get bRigidBodyEnabled()) {blendType=Linear})
   Save cached pose 'Base_pose'(Pose=Input Pose()) {cacheName=Base_pose}
  ]}
 EventGraph {nodes [
   Event BlueprintUpdateAnimation()
   > Set not in FE(not in FE:bool=NOT Boolean(A=Get bIsInFrontEnd())) -> (Output_Get:bool)
   TryGetPawnOwner()
  ]}
]
variables [not in FE {type=boolean}]
